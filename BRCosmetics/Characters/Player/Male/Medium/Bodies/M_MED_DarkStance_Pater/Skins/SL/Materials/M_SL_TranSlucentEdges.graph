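# Masked character material for the DarkStance Pater SL (SuperLevel) skin that cuts animated, noisy smoke-like edges into the body.
# PreFX output runs through a static-switched S29 tier function (T3 or T1/T2) and the Styx cosmetic function, gets an Opacity Mask override, then enters PostFX.
# Opacity combines a smoothstepped gradient measured from BottomLoc with panning, distorted noise, thresholded per channel and confined by the SmokeArea texture.
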Material /BRCosmetics/Characters/Player/Male/Medium/Bodies/M_MED_DarkStance_Pater/Skins/SL/Materials/M_SL_TranSlucentEdges
graph {
 properties {
  bCastRayTracedShadows=false
  bUseMaterialAttributes=true
  bUsedWithMeshDeformer=true
  bUsedWithMorphTargets=true
  bUsedWithSkeletalMesh=true
  blendMode=BLEND_Masked
  pixelDepthOffsetMode=PDOM_Legacy
  shadingModel=MSM_FromMaterialExpression
  usedShadingModels=Default Lit | Subsurface | Clear Coat | Subsurface Profile | Cloth
 }
 nodes [
  n1 = TextureSampleParameter2D() {
   group=OpacityControl
   parameterName=SmokeArea
   samplerType=SAMPLERTYPE_Masks
   texture=/BRCosmetics/Characters/Player/Male/Medium/Bodies/M_MED_DarkStance_Pater/Textures/T_DarkStanceSmokeArea
  }
  LinearInterpolate(Alpha=@n1.R) {constB=0.5}
  n2 = MaterialFunctionCall() {materialFunction=&MF_CharacterParent_2023_PreFX}
  n3 = Constant() {r=1}
  n4 = ComponentMask(None=@n2.M) {g=true}
  n5 = MaterialFunctionCall!(
   Material=StaticSwitchParameter(
    True=MaterialFunctionCall!(In=@n2.Result, SLMask_Primary=@n3, Gmap=@n4) {materialFunction=/BRCosmetics/Characters/Player/Male/Medium/Bodies/M_MED_StunningMountain_Armor/Skins/SL/Materials/MF_SL_S29_T3}.Result
    False=MaterialFunctionCall!(In=@n2.Result, SLMaskRGB=@n3, Gmap=@n4) {materialFunction=/BRCosmetics/Characters/Player/Male/Medium/Bodies/M_MED_StunningMountain_Armor/Skins/SL/Materials/MF_SL_S29_T1nT2}.Result
   ) {group=01_SL_S29, parameterName=SL_S29_T3, sortPriority=0}
   Tech Art Mask=@n2.Tech Art Mask
   PreSkinnedLocation=NamedRerouteDeclaration(None=@n2.Pre skinned Local Position) {name=PreSkinnedLoc, nodeColor=<rgba 1 0.1875 0 1>}
  ) {materialFunction=/BRCosmetics/Characters/Player/Male/Medium/Bodies/M_MED_DarkStance_Pater/Materials/MF_StyxCosmetic_Generic}
  MaterialFunctionCall(
   Material=SetMaterialAttributes!(MaterialAttributes=@n5.Result, Opacity Mask=NamedRerouteUsage()) {attributeSetTypes [#1]}
   M=@n2.M
   Pre Skinned Local Position=@n2.M
   Distance Blend=@n2.M
   Fresnel=@n2.M
  ) {materialFunction=&MF_CharacterParent_2023_PostFX}
  MaterialFunctionCall!(MA_In=@n2.Result, Gmap=@n4, TAMask=@n2.Result) {materialFunction=/BRCosmetics/Characters/Player/Male/Medium/Bodies/M_MED_StunningMountain_Armor/Skins/SL/Materials/MF_SuperLevel_S29_Master}
  n6 = ComponentMask(None=Subtract(A=NamedRerouteUsage(), B=VectorParameter() {group=OpacityControl, parameterName=BottomLoc, defaultValue=<rgba 0 0 10 1>}.RGB)) {b=true}
  n7 = Divide(A=@n6, B=ScalarParameter() {defaultValue=80, group=OpacityControl, parameterName=OpacityGradientUnit01}) {constB=50}
  n8 = SmoothStep(
   Min=ScalarParameter() {defaultValue=0.4, group=OpacityControl, parameterName=GradientMin01}
   Max=ScalarParameter() {defaultValue=0.7, group=OpacityControl, parameterName=GradientMax01}
   Value=NamedRerouteDeclaration(None=Saturate(None=@n7)) {name=OpacityGradient01, nodeColor=<rgba 1 0 0.2812 1>}
  )
  n9 = Multiply(A=@n1.R)
  n10 = Multiply(A=TextureCoordinate(), B=ScalarParameter() {defaultValue=10, group=OpacityControl, parameterName=NoiseTilling}) {constB=30}
  n11 = Multiply(A=NamedRerouteUsage(), B=ScalarParameter() {defaultValue=0.1, parameterName=DIstort_Intensity}) {constB=0.2}
  n12 = ScalarParameter() {group=OpacityControl, parameterName=OpacityNoiseSpeedX}
  n13 = ScalarParameter() {defaultValue=-0.2, group=OpacityControl, parameterName=OpacityNoiseSpeedY}
  n14 = TextureSampleParameter2D(UVs=Panner(Coordinate=Add(A=@n10, B=@n11), Speed=AppendVector(A=@n12, B=@n13))) {
   group=OpacityControl
   parameterName=OpacityNoise
   samplerType=SAMPLERTYPE_Masks
   texture=/BRCosmetics/Characters/Player/Male/Medium/Bodies/M_MED_DarkStance_Pater/Textures/T_DarkStanceNoiseMask
  }
  n15 = ComponentMask(None=@n14.RGB) {b=true, g=true, r=true}
  n16 = NamedRerouteUsage()
  n17 = Saturate(None=Subtract(A=@n15, B=OneMinus(None=@n16)))
  n18 = ComponentMask(None=@n15) {b=true}
  n19 = OneMinus(None=NamedRerouteUsage())
  n20 = Add(A=Add(A=ComponentMask(None=@n17) {r=true}, B=ComponentMask(None=@n17) {g=true}), B=Saturate(None=Subtract(A=@n18, B=@n19)))
  n21 = Multiply(A=@n1.G, B=Add(A=Saturate(None=@n9), B=Saturate(None=@n20)))
  NamedRerouteDeclaration(None=Saturate(None=Add(A=@n8, B=@n21))) {name=Opacity_OUT, nodeColor=<rgba 0 0.1094 1 1>}
  n22 = ScalarParameter() {defaultValue=40, group=OpacityControl, parameterName=OpacityGradientUnit02}
  NamedRerouteDeclaration(None=Saturate(None=Divide(A=@n6, B=@n22) {constB=50})) {name=OpacityGradient02, nodeColor=<rgba 1 0 0.2812 1>}
  NamedRerouteDeclaration(None=@n5.UV_Distort) {name=UVDistort, nodeColor=<rgba 0.8281 0 1 1>}
  n23 = ScalarParameter() {defaultValue=10, group=OpacityControl, parameterName=OpacityGradientUnit03}
  NamedRerouteDeclaration(None=Saturate(None=Divide(A=@n6, B=@n23) {constB=50})) {name=OpacityGradient03, nodeColor=<rgba 1 0 0.2812 1>}
 ]
}
